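# Physics asset for the BrineGleam female medium player body cosmetic, covering the pelvis and a three-bone dynamic tail.
# One large pelvis capsule anchors the tail, and each tail bone has a small identical capsule.
# Constraints chain the tail from the pelvis to the tip, with swing limits tightening at each joint and twist limits set to zero.
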PhysicsAsset /BRCosmetics/Characters/Player/Female/Medium/Bodies/F_MED_BrineGleam/Meshes/F_MED_BrineGleam_Physics
bodies {
 dyn_tail_01 [{shapeType=Capsule, length=3, radius=2, shapeName=dyn_tail_01_capsule, rotation=<rot 90 0 0>}]
 dyn_tail_02 [{shapeType=Capsule, length=3, radius=2, shapeName=dyn_tail_02_capsule, rotation=<rot 90 0 0>}]
 dyn_tail_03 [{shapeType=Capsule, length=3, radius=2, shapeName=dyn_tail_03_capsule, rotation=<rot 90 0 0>}]
 pelvis [{
   shapeType=Capsule
   length=17.4474
   radius=15.0713
   shapeName=pelvis_capsule
   center=<-4.4669 -5.6715 -0.008>
   rotation=<rot 0 176.6031 -177.8218>
  }]
}
constraints [
 {bone1Name=dyn_tail_01, bone2Name=pelvis, swing1LimitDegrees=25, swing2LimitDegrees=25, twistLimitDegrees=0}
 {bone1Name=dyn_tail_02, bone2Name=dyn_tail_01, swing1LimitDegrees=20, swing2LimitDegrees=20, twistLimitDegrees=0}
 {bone1Name=dyn_tail_03, bone2Name=dyn_tail_02, swing1LimitDegrees=15, swing2LimitDegrees=15, twistLimitDegrees=0}
]
