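# Post-process style AnimBP for the F_MED_BrineGleam player body, adding secondary motion to the pose copied from the parent mesh.
# RigidBody physics feeds AnimDynamics on pants and torso bones, the torso using stacked angular and linear spring passes.
# A deform Control Rig follows, driven by long-backbling detection and a tail adjustment for two emotes.
# Tail bones blend to DefaultSlot via a front-end idle curve; skydive floating switches a gravity variable.
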AnimBlueprint /BRCosmetics/Characters/Player/Female/Medium/Bodies/F_MED_BrineGleam/Meshes/F_MED_BrineGleam_AnimBP
target_skeleton=&/F_MED_BrineGleam_Skeleton
graphs [
 AnimGraph {nodes [
   n1 = RigidBody!(
    ComponentPose=Local To Component(LocalPose=Copy Pose From Mesh() {bUseAttachedParent=true})
    Alpha
    OverrideWorldGravity:Vector="0,0,980"
    ExternalForce
   ) -> (Pose) {
    bOverrideWorldGravity=true
    overridePhysicsAsset=&/F_MED_BrineGleam_Physics
    simulationSpace=BaseBoneSpace
    baseBoneRef {boneName=spine_01}
    componentAppliedLinearAccClamp=<0 0 0>
    overrideWorldGravity {z=980}
    simSpaceSettings {dampingAlpha=0, maxAngularAcceleration=15, maxAngularVelocity=3, maxLinearAcceleration=1250, maxLinearVelocity=250, worldAlpha=1}
   }
   n2 = Anim Dynamics - Bone: dyn_l_pantsRotate_01(ComponentPose=@n1, Alpha=1, ExternalForce="0,0,0") {
    angularDampingOverride=0.9999
    bOverrideAngularDamping=true
    bUseGravityOverride=true
    boundBone {boneName=dyn_l_pantsRotate_01}
    componentAppliedLinearAccClamp=<1500 1500 1500>
    componentLinearAccScale=<0.1 0.1 0.1>
    componentLinearVelScale=<0.3 0.3 0.3>
    gravityOverride {z=-980}
    physicsBodyDefinitions [{
      boundBone {boneName=dyn_l_pantsRotate_01}
      boxExtents=<15 30 30>
      constraintSetup {angularLimitsMax=<0 7.5 7.5>, angularLimitsMin=<0 -7.5 -7.5>, angularTarget=<0 0 0>}
      localJointOffset=<0.5 0 0>
     }]
   }
   n3 = Anim Dynamics - Bone: dyn_r_pantsRotate_01(ComponentPose=@n2, Alpha=1, ExternalForce="0,0,0") {
    angularDampingOverride=0.9999
    bOverrideAngularDamping=true
    bUseGravityOverride=true
    boundBone {boneName=dyn_r_pantsRotate_01}
    componentAppliedLinearAccClamp=<1500 1500 1500>
    componentLinearAccScale=<0.1 0.1 0.1>
    componentLinearVelScale=<0.3 0.3 0.3>
    gravityOverride {z=-980}
    physicsBodyDefinitions [{
      boundBone {boneName=dyn_r_pantsRotate_01}
      boxExtents=<15 30 30>
      constraintSetup {angularLimitsMax=<0 7.5 7.5>, angularLimitsMin=<0 -7.5 -7.5>, angularTarget=<0 0 0>}
      localJointOffset=<0.5 0 0>
     }]
   }
   n4 = "Transform (Modify) Bone - Bone: dyn_torso_l"(ComponentPose=@n3, Alpha=1, Translation="0,0,1", Rotation="0,0,0", Scale="1,1,1") {translationMode=BMM_Additive, translationSpace=BCS_BoneSpace, boneToModify {boneName=dyn_torso_l}, translation {z=1}}
   n5 = "Transform (Modify) Bone - Bone: dyn_torso_r"(ComponentPose=@n4, Alpha=1, Translation="0,0,1", Rotation="0,0,0", Scale="1,1,1") {translationMode=BMM_Additive, translationSpace=BCS_BoneSpace, boneToModify {boneName=dyn_torso_r}, translation {z=1}}
   n6 = Anim Dynamics - Bone: dyn_torso_r!(ComponentPose=@n5, Alpha=1, GravityOverride:Vector="0,0,-980", ExternalForce="0,0,0") -> (Pose) {
    angularBiasOverride=0.9999
    angularDampingOverride=0.999
    angularSpringConstant=4000
    bAngularSpring=true
    bGravityOverrideInSimSpace=true
    bOverrideAngularBias=true
    bOverrideAngularDamping=true
    bUseGravityOverride=true
    linearDampingOverride=0.9999
    linearSpringConstant=100
    simulationSpace=RootRelative
    boundBone {boneName=dyn_torso_r}
    gravityOverride {z=-980}
    physicsBodyDefinitions [{
      boundBone {boneName=dyn_torso_r}
      boxExtents=<21 21 21>
      constraintSetup {angularLimitsMax=<0 1 5>, angularLimitsMin=<0 -10 -5>, angularTarget=<1 0 0>}
      localJointOffset=<2 0 0>
     }]
    relativeSpaceBone {boneName=spine_04}
    simSpaceSettings {maxAngularAcceleration=45, maxAngularVelocity=8, simSpaceAngularAlpha=0.3}
   }
   n7 = Anim Dynamics - Bone: dyn_torso_l!(ComponentPose=@n6, Alpha=1, GravityOverride:Vector="0,0,-980", ExternalForce="0,0,0") -> (Pose) {
    angularBiasOverride=0.9999
    angularDampingOverride=0.999
    angularSpringConstant=4000
    bAngularSpring=true
    bGravityOverrideInSimSpace=true
    bOverrideAngularBias=true
    bOverrideAngularDamping=true
    bUseGravityOverride=true
    linearDampingOverride=0.9999
    linearSpringConstant=100
    simulationSpace=RootRelative
    boundBone {boneName=dyn_torso_l}
    gravityOverride {z=-980}
    physicsBodyDefinitions [{
      boundBone {boneName=dyn_torso_l}
      boxExtents=<20 20 20>
      constraintSetup {angularLimitsMax=<0 1 5>, angularLimitsMin=<0 -10 -5>, angularTarget=<1 0 0>}
      localJointOffset=<2 0 0>
     }]
    relativeSpaceBone {boneName=spine_04}
    simSpaceSettings {maxAngularAcceleration=45, maxAngularVelocity=8, simSpaceAngularAlpha=0.3}
   }
   n8 = Anim Dynamics - Bone: dyn_torso_r(ComponentPose=@n7, Alpha=1, GravityOverride="0,0,-200", ExternalForce="0,0,0") {
    angularBiasOverride=1
    angularDampingOverride=0.99
    angularSpringConstant=10000
    bGravityOverrideInSimSpace=true
    bLinearSpring=true
    bOverrideLinearDamping=true
    bUseGravityOverride=true
    linearDampingOverride=0.9999
    linearSpringConstant=180
    simulationSpace=RootRelative
    boundBone {boneName=dyn_torso_r}
    gravityOverride {z=-200}
    physicsBodyDefinitions [{
      boundBone {boneName=dyn_torso_r}
      boxExtents=<80 80 80>
      constraintSetup {angularLimitsMax=<0 0 0>, angularLimitsMin=<0 0 0>, angularTarget=<1 0 0>, linearAxesMax=<0 0.5 1>, linearAxesMin=<0 -0.5 -0.3>}
      localJointOffset=<0.2 0 0>
     }]
    relativeSpaceBone {boneName=spine_04}
    simSpaceSettings {maxAngularAcceleration=400, maxAngularVelocity=200, simSpaceAngularAlpha=0.1}
   }
   n9 = Anim Dynamics - Bone: dyn_torso_l(ComponentPose=@n8, Alpha=1, GravityOverride="0,0,-200", ExternalForce="0,0,0") {
    angularBiasOverride=1
    angularDampingOverride=0.99
    angularSpringConstant=10000
    bGravityOverrideInSimSpace=true
    bLinearSpring=true
    bOverrideLinearDamping=true
    bUseGravityOverride=true
    linearDampingOverride=0.9999
    linearSpringConstant=180
    simulationSpace=RootRelative
    boundBone {boneName=dyn_torso_l}
    gravityOverride {z=-200}
    physicsBodyDefinitions [{
      boundBone {boneName=dyn_torso_l}
      boxExtents=<90 90 90>
      constraintSetup {angularLimitsMax=<0 0 0>, angularLimitsMin=<0 0 0>, angularTarget=<1 0 0>, linearAxesMax=<0 0.5 1>, linearAxesMin=<0 -0.5 -0.3>}
      localJointOffset=<0.2 0 0>
     }]
    relativeSpaceBone {boneName=spine_04}
    simSpaceSettings {maxAngularAcceleration=400, maxAngularVelocity=200, simSpaceAngularAlpha=0.1}
   }
   n10 = Control Rig!(
    Source=Component To Local(ComponentPose=@n9)
    Alpha
    Cape_Enabled:bool=Get Is Wearing Cape() -> (Is Wearing Cape:bool)
    AdjustTailToEmote:double=Get AdjustTailToEmotes() -> (AdjustTailToEmotes:double)
   ) -> (Pose) {controlRigAssetReference {blueprintRigClass=&/F_MED_BrineGleam_Deform_CtrlRig.F_MED_BrineGleam_Deform_CtrlRig_C}}
   n11 = Layered blend per bone(BasePose=@n10, BlendPoses_0=Slot 'DefaultSlot'(), BlendWeights_0=Get customIdleIsPlayingFloat() -> (customIdleIsPlayingFloat:float)) {layerSetup [{branchFilters [{boneName=dyn_tail_01}, {boneName=dyn_tail_02}, {boneName=dyn_tail_03}]}]}
   Output Pose(Result=@n11) {layerGroup=None}
  ]}
 EventGraph {nodes [
   Event BlueprintUpdateAnimation()
   > n1 = Sequence()
   > Set!(
    execute
    GravityOverrideValue_X:double=0
    GravityOverrideValue_Y:double=0
    GravityOverrideValue_Z:double=SelectFloat(A=980, B=0, bPickA=Get bIsSkydiveFloating())
   ) -> (then, Output_Get:Vector)
   > Branch(Condition=Get bIsInFrontEnd())
   > Set customIdleIsPlayingFloat(customIdleIsPlayingFloat:float=GetCurveValue(self=GetOwnerFrontEndAnimInstance(), CurveName=Frontend_Idle_Bypass)) -> (Output_Get:float)
   n2 = GetAnimInstance(self=GetOwnerSkeletalMeshComponent())
   n3 = Montage_IsPlaying(self=@n2, Montage=/BRCosmetics/Animation/Game/MainPlayer/Emotes/Cauldron/Emote_Cauldron_CMF_M)
   n4 = Montage_IsPlaying(self=@n2, Montage=/BRCosmetics/Animation/Game/MainPlayer/Emotes/Scrub_A_Dub/Emote_ScrubaDub_CMF_M)
   Set AdjustTailToEmotes(execute=@n1.then_1, AdjustTailToEmotes:double="To Float (Boolean)"(InBool=OR Boolean(A=@n3, B=@n4))) -> (Output_Get:double)
   Event OnCharacterCustomizationLoaded()
   n5 = CheckForBackbling()
   > Set Is Wearing BP(Is Wearing BP:bool=OR Boolean!(A=@n5.bIsBackblingACape, B=@n5.bIsBackblingLong, C:bool=@n5.bIsBackblingLong) -> (ReturnValue)) -> (Output_Get:bool)
   > Set Is Wearing Cape(Is Wearing Cape=@n5.bIsBackblingLong)
  ]}
]
variables [
 AdjustTailToEmotes {type=number}
 customIdleIsPlayingFloat {type=number}
 GravityOverrideValue {type=Vector}
 Is Wearing BP {type=boolean}
 Is Wearing Cape {type=boolean}
]
