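# Physics asset for the F_MED_BrineGleam face accessory mesh, setting up secondary motion for its dynamic ear and hair bones around the head.
# Every body is a capsule, including a large one on the head.
# A four-bone main ear chain hangs from the head and forks into three-bone left and right ear chains, alongside a separate three-bone hair chain.
# Constraints allow limited swing with very little twist, and both chain roots are nearly locked to one swing axis.
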PhysicsAsset /BRCosmetics/Characters/Player/Female/Medium/Bodies/F_MED_BrineGleam/Meshes/Parts/F_MED_BrineGleam_FaceAcc_Physics
bodies {
 dyn_earMain_01 [{shapeType=Capsule, length=10, radius=5, shapeName=dyn_earMain_01_capsule, rotation=<rot -0.0001 0 -95>}]
 dyn_earMain_02 [{shapeType=Capsule, length=10, radius=5, shapeName=dyn_earMain_01_capsule, rotation=<rot -0.0001 0 -95>}]
 dyn_earMain_03 [{shapeType=Capsule, length=10, radius=5, shapeName=dyn_earMain_03_capsule, rotation=<rot 0 0 -90>}]
 dyn_earMain_04 [{shapeType=Capsule, length=9, radius=6, shapeName=dyn_earMain_04_capsule, rotation=<rot 0 0 -90>}]
 dyn_ear_l_01 [{shapeType=Capsule, length=7.5, radius=4, shapeName=dyn_ear_l_01_capsule, rotation=<rot 90 0 0>}]
 dyn_ear_l_02 [{shapeType=Capsule, length=6, radius=3, shapeName=dyn_ear_l_02_capsule, rotation=<rot 90 0 0>}]
 dyn_ear_l_03 [{shapeType=Capsule, length=5, radius=2.5, shapeName=dyn_ear_l_03_capsule, center=<2.8581 0 0>, rotation=<rot 90 0 0>}]
 dyn_ear_r_01 [{shapeType=Capsule, length=7.5, radius=4, shapeName=dyn_ear_r_01_capsule, rotation=<rot 90 0 0>}]
 dyn_ear_r_02 [{shapeType=Capsule, length=6, radius=3, shapeName=dyn_ear_r_02_capsule, rotation=<rot 90 0 0>}]
 dyn_ear_r_03 [{shapeType=Capsule, length=5, radius=2.5, shapeName=dyn_ear_r_03_capsule, center=<2.8581 0 0>, rotation=<rot 90 0 0>}]
 dyn_hair_01 [{shapeType=Capsule, length=4.9378, radius=5.0099, shapeName=dyn_hair_01_capsule, rotation=<rot 0.0004 0 -89.9999>}]
 dyn_hair_02 [{shapeType=Capsule, length=8.5749, radius=3.3196, shapeName=dyn_hair_02_capsule, rotation=<rot 0 0 -90>}]
 dyn_hair_03 [{shapeType=Capsule, length=4.3018, radius=4.8833, shapeName=dyn_hair_03_capsule, rotation=<rot 0 0 -90>}]
 head [{shapeType=Capsule, length=19.125, radius=16.125, shapeName=head_capsule, center=<7.024 -5.3371 0>, rotation=<rot 0 179.9994 -180>}]
}
constraints [
 {bone1Name=dyn_earMain_01, bone2Name=head, swing1LimitDegrees=10, swing2LimitDegrees=0.1, twistLimitDegrees=0.1}
 {bone1Name=dyn_earMain_02, bone2Name=dyn_earMain_01, swing1LimitDegrees=15, swing2LimitDegrees=40, twistLimitDegrees=1}
 {bone1Name=dyn_earMain_03, bone2Name=dyn_earMain_02, swing1LimitDegrees=15, swing2LimitDegrees=20, twistLimitDegrees=1}
 {bone1Name=dyn_earMain_04, bone2Name=dyn_earMain_03, swing1LimitDegrees=20, swing2LimitDegrees=15, twistLimitDegrees=1}
 {bone1Name=dyn_ear_l_01, bone2Name=dyn_earMain_04, swing1LimitDegrees=15, swing2LimitDegrees=15, twistLimitDegrees=1}
 {bone1Name=dyn_ear_l_02, bone2Name=dyn_ear_l_01, swing1LimitDegrees=25, swing2LimitDegrees=10, twistLimitDegrees=1}
 {bone1Name=dyn_ear_l_03, bone2Name=dyn_ear_l_02, swing1LimitDegrees=30, swing2LimitDegrees=15, twistLimitDegrees=1}
 {bone1Name=dyn_ear_r_01, bone2Name=dyn_earMain_04, swing1LimitDegrees=10, swing2LimitDegrees=15, twistLimitDegrees=1}
 {bone1Name=dyn_ear_r_02, bone2Name=dyn_ear_r_01, swing1LimitDegrees=10, swing2LimitDegrees=25, twistLimitDegrees=1}
 {bone1Name=dyn_ear_r_03, bone2Name=dyn_ear_r_02, swing1LimitDegrees=15, swing2LimitDegrees=30, twistLimitDegrees=1}
 {bone1Name=dyn_hair_01, bone2Name=head, swing1LimitDegrees=15, swing2LimitDegrees=0.1, twistLimitDegrees=0.1}
 {bone1Name=dyn_hair_02, bone2Name=dyn_hair_01, swing1LimitDegrees=17.5, swing2LimitDegrees=5, twistLimitDegrees=2}
 {bone1Name=dyn_hair_03, bone2Name=dyn_hair_02, swing1LimitDegrees=20, swing2LimitDegrees=5, twistLimitDegrees=2}
]
