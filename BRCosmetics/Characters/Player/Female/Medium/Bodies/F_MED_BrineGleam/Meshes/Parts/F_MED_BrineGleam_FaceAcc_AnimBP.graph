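# Animation Blueprint for the BrineGleam female face accessory, adding physics-driven secondary motion to the pose copied from the attached parent mesh.
# Two RigidBody nodes simulate in head-bone space with limited velocities and accelerations, one using a with-cape physics asset.
# A Blend Poses by bool node chooses between them.
# Each update sets a backbling flag and a gravity variable that turns upward during skydive floating, though neither shows a link into the AnimGraph.
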AnimBlueprint /BRCosmetics/Characters/Player/Female/Medium/Bodies/F_MED_BrineGleam/Meshes/Parts/F_MED_BrineGleam_FaceAcc_AnimBP
target_skeleton=&/F_MED_BrineGleam_FaceAcc_Skeleton
graphs [
 AnimGraph {nodes [
   n1 = RigidBody!(
    ComponentPose=Local To Component(LocalPose=Copy Pose From Mesh() {bUseAttachedParent=true})
    Alpha
    OverrideWorldGravity:Vector="0,0,-980"
    ExternalForce
   ) -> (Pose) {
    bClampLinearTranslationLimitToRefPose=true
    bOverrideWorldGravity=true
    overridePhysicsAsset=&/F_MED_BrineGleam_FaceAcc_Physics
    simulationSpace=BaseBoneSpace
    baseBoneRef {boneName=head}
    componentAppliedLinearAccClamp=<0 0 0>
    overrideWorldGravity {z=-980}
    simSpaceSettings {
     dampingAlpha=0
     maxAngularAcceleration=25
     maxAngularVelocity=5
     maxLinearAcceleration=1200
     maxLinearVelocity=300
     velocityScaleZ=0.5
     worldAlpha=1
    }
   }
   n2 = RigidBody(ComponentPose=Local To Component(LocalPose=Copy Pose From Mesh() {bUseAttachedParent=true}), OverrideWorldGravity="0,0,-980") {
    bClampLinearTranslationLimitToRefPose=true
    bOverrideWorldGravity=true
    overridePhysicsAsset=&/F_MED_BrineGleam_FaceAcc_withCape_Physics
    simulationSpace=BaseBoneSpace
    baseBoneRef {boneName=head}
    componentAppliedLinearAccClamp=<0 0 0>
    overrideWorldGravity {z=-980}
    simSpaceSettings {
     dampingAlpha=0
     maxAngularAcceleration=25
     maxAngularVelocity=5
     maxLinearAcceleration=1200
     maxLinearVelocity=300
     velocityScaleZ=0.5
     worldAlpha=1
    }
   }
   Output Pose(Result=Blend Poses by bool(BlendPose_0=Component To Local(ComponentPose=@n2), BlendPose_1=Component To Local(ComponentPose=@n1))) {layerGroup=None}
  ]}
 EventGraph {nodes [
   Event BlueprintUpdateAnimation()
   > Set!(
    execute
    GravityOverrideValue_X:double=0
    GravityOverrideValue_Y:double=0
    GravityOverrideValue_Z:double=SelectFloat(A=300, B=-980, bPickA=Get bIsSkydiveFloating())
   ) -> (then, Output_Get:Vector)
   > Set Is Wearing Backbling(Is Wearing Backbling=CheckForBackbling())
  ]}
]
variables [
 GravityOverrideValue {type=Vector}
 Is Wearing Backbling {type=boolean}
]
